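# Top-level State Tree for the large Spire boss, sequencing an intro, a pre-combat state and combat.
# Intro runs while Finished Intro Sequence is false, playing a linked cinematic tree, then changing a State Tree bool and returning to Root.
# PreCombat sets spire state, updates shields and waits, with target-validity checks gating entry and the switch to linked Combat.
# Combat returns to PreCombat on a target check, only when neither active-ability nor summoning tags are present.
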StateTree /SpireBoss/StateTrees/ST_Large_Spire
states [
 Root {
  id=#1
  type=STATE
  children [#2, #3, #4]
  tasks []
  transitions []
 }
 Intro {
  id=#2
  type=STATE
  children [#5, #6]
  enter_conditions [{
    id=#7
    description=Parameters.Finished Intro Sequence is False
    properties {
     iD=#8
     instance {_structType=/Script/StateTreeModule.StateTreeCompareBoolConditionInstanceData, bLeft=false, bRight=false}
     node {
      name=None
      _structType=/Script/StateTreeModule.StateTreeCompareBoolCondition
      bInvert=false
      deltaIndent=0
      evaluationMode=Evaluated
      operand=And
      bindingsBatch {value=65535}
      executionRuntimeTemplateIndex {value=65535}
      instanceDataHandle {index=65535, source=None, stateHandle {index=65535}}
      instanceTemplateIndex {value=65535}
      outputBindingsBatch {value=65535}
     }
    }
   }]
  tasks [{
    id=#9
    description=State Tree Task Spires Set State
    properties {
     iD=#10
     instanceObject=&self:StateTreeEditorData_0.StateTreeState_2.StateTreeTask_Spires_SetState_C_0
     node {
      name=None
      _structType=&StateTreeBlueprintTaskWrapper
      taskClass=/SpireBoss/StateTrees/Tasks/StateTreeTask_Spires_SetState.StateTreeTask_Spires_SetState_C
     }
    }
   }]
  transitions []
 }
 PreCombat {
  id=#3
  type=STATE
  enter_conditions [{
    id=#11
    description=State Tree Conditional Spires Has Valid Ability Target
    properties {
     iD=#12
     instanceObject=&self:StateTreeEditorData_0.StateTreeState_1.StateTreeConditional_Spires_HasValidAbilityTarget_C_1
     node {
      name=None
      _structType=/Script/StateTreeModule.StateTreeBlueprintConditionWrapper
      conditionClass=/SpireBoss/StateTrees/Conditionals/StateTreeConditional_Spires_HasValidAbilityTarget.StateTreeConditional_Spires_HasValidAbilityTarget_C
      deltaIndent=0
      evaluationMode=Evaluated
      operand=And
      bindingsBatch {value=65535}
      executionRuntimeTemplateIndex {value=65535}
      instanceDataHandle {index=65535, source=None, stateHandle {index=65535}}
      instanceTemplateIndex {value=65535}
      outputBindingsBatch {value=65535}
     }
    }
   }]
  tasks [
   {
    id=#13
    description=Delay Forever
    properties {
     iD=#14
     instance {_structType=&StateTreeDelayTaskInstanceData, bRunForever=true, duration=1}
     node {name=None, _structType=&StateTreeDelayTask}
    }
   }
   {
    id=#15
    description="Debug Text \"PRECOMBAT\""
    properties {
     iD=#16
     instance {_structType=/Script/StateTreeModule.StateTreeDebugTextTaskInstanceData, bEnabled=false, bindableText="", referenceActor=null}
     node {
      name=None
      _structType=/Script/StateTreeModule.StateTreeDebugTextTask
      bCanEditConsideredForCompletion=false
      bConsideredForCompletion=false
      bHasTaskCompletionDelegateDispatcher=false
      bTaskEnabled=false
      deprecatedVersion=1
      fontScale=1
      text=PRECOMBAT
      transitionHandlingPriority=Normal
      bindingsBatch {value=65535}
      executionRuntimeTemplateIndex {value=65535}
      instanceDataHandle {index=65535, source=None, stateHandle {index=65535}}
      instanceTemplateIndex {value=65535}
      offset=<0 0 0>
      outputBindingsBatch {value=65535}
      textColor=<rgba 1 0 0.9765 1>
     }
    }
   }
   {
    id=#17
    description=State Tree Task Spires Set State
    properties {
     iD=#18
     instanceObject=&self:StateTreeEditorData_0.StateTreeState_1.StateTreeTask_Spires_SetState_C_0
     node {
      name=None
      _structType=&StateTreeBlueprintTaskWrapper
      taskClass=&StateTreeTask_Spires_SetState_C
     }
    }
   }
   {
    id=#19
    description=STBTT Spire Update Shields
    properties {
     iD=#20
     instanceObject=&self:StateTreeEditorData_0.StateTreeState_1.STBTT_Spire_UpdateShields_C_0
     node {
      name=None
      _structType=&StateTreeBlueprintTaskWrapper
      taskClass=/SpireBoss/StateTrees/Tasks/STBTT_Spire_UpdateShields.STBTT_Spire_UpdateShields_C
     }
    }
   }
  ]
  transitions [{
    id=#21
    trigger=ON_TICK
    conditions [{
      id=#22
      description=State Tree Conditional Spires Has Valid Ability Target
      properties {
       iD=#23
       instanceObject=&self:StateTreeEditorData_0.StateTreeState_1.StateTreeConditional_Spires_HasValidAbilityTarget_C_0
       node {
        name=None
        _structType=&StateTreeBlueprintConditionWrapper
        conditionClass=&StateTreeConditional_Spires_HasValidAbilityTarget_C
        deltaIndent=0
        evaluationMode=Evaluated
        operand=And
        bindingsBatch {value=65535}
        executionRuntimeTemplateIndex {value=65535}
        instanceDataHandle {index=65535, source=None, stateHandle {index=65535}}
        instanceTemplateIndex {value=65535}
        outputBindingsBatch {value=65535}
       }
      }
     }]
    state {id=#4, name=Combat}
   }]
 }
 Combat {
  id=#4
  type=LINKED_ASSET
  transitions [{
    id=#24
    priority=null
    trigger=ON_TICK
    conditions [
     {
      id=#25
      description="(Fortnite) Actor Has Matching Gameplay Tag"
      properties {
       executionRuntimeDataObject=null
       expressionIndent=0
       expressionOperand=And
       iD=#26
       instanceObject=null
       executionRuntimeData {}
       instance {_structType=/Script/FortniteAIServer.GameplayTagActorHasMatchingGameplayTagInstanceData, actor=null, tagToCheck {tagName=Ability.SB.Active}}
       node {
        name=None
        _structType=/Script/FortniteAIServer.GameplayTagActorHasMatchingGameplayTagCondition
        bInvert=true
        deltaIndent=0
        evaluationMode=Evaluated
        operand=And
        bindingsBatch {value=65535}
        executionRuntimeTemplateIndex {value=65535}
        instanceDataHandle {index=65535, source=None, stateHandle {index=65535}}
        instanceTemplateIndex {value=65535}
        outputBindingsBatch {value=65535}
       }
      }
     }
     {
      id=#27
      description=State Tree Conditional Spires Has Valid Ability Target
      properties {
       executionRuntimeDataObject=null
       expressionIndent=0
       expressionOperand=And
       iD=#28
       instanceObject=&self:StateTreeEditorData_0.StateTreeState_3.StateTreeConditional_Spires_HasValidAbilityTarget_C_0
       executionRuntimeData {}
       instance {}
       node {
        name=None
        _structType=&StateTreeBlueprintConditionWrapper
        conditionClass=&StateTreeConditional_Spires_HasValidAbilityTarget_C
        deltaIndent=0
        evaluationMode=Evaluated
        operand=And
        bindingsBatch {value=65535}
        executionRuntimeTemplateIndex {value=65535}
        instanceDataHandle {index=65535, source=None, stateHandle {index=65535}}
        instanceTemplateIndex {value=65535}
        outputBindingsBatch {value=65535}
       }
      }
     }
     {
      id=#29
      description="(Fortnite) Actor Has Matching Gameplay Tag"
      properties {
       executionRuntimeDataObject=null
       expressionIndent=0
       expressionOperand=And
       iD=#30
       instanceObject=null
       executionRuntimeData {}
       instance {
        _structType=&GameplayTagActorHasMatchingGameplayTagInstanceData
        actor=null
        tagToCheck {tagName=Ability.SB.Summon.Summoning}
       }
       node {
        name=None
        _structType=&GameplayTagActorHasMatchingGameplayTagCondition
        bInvert=true
        deltaIndent=0
        evaluationMode=Evaluated
        operand=And
        bindingsBatch {value=65535}
        executionRuntimeTemplateIndex {value=65535}
        instanceDataHandle {index=65535, source=None, stateHandle {index=65535}}
        instanceTemplateIndex {value=65535}
        outputBindingsBatch {value=65535}
       }
      }
     }
    ]
    state {id=#3, name=PreCombat, linkType=GOTO_STATE}
   }]
 }
 IntroCinematic {
  id=#5
  type=LINKED_ASSET
  transitions [{
    id=#31
    priority=null
    trigger=ON_STATE_COMPLETED
    conditions []
    state {id=#32, name=Root, linkType=NEXT_STATE}
   }]
 }
 MarkIntroCinematicAsCompleted {
  id=#6
  type=STATE
  tasks [{
    id=#33
    description=State Tree Task Spires Change State Tree Bool
    properties {
     iD=#34
     instanceObject=&self:StateTreeEditorData_0.StateTreeState_5.StateTreeTask_Spires_ChangeStateTreeBool_C_0
     node {
      name=None
      _structType=&StateTreeBlueprintTaskWrapper
      taskClass=/SpireBoss/StateTrees/States/StateTreeTask_Spires_ChangeStateTreeBool.StateTreeTask_Spires_ChangeStateTreeBool_C
     }
    }
   }]
  transitions [{id=#35, trigger=ON_STATE_COMPLETED, state {id=#1, name=Root}}]
 }
]
