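# StateTree for the Large Spire boss combat phase, which applies an initial-combat cooldown effect, runs the encounter, and sets spire state and shields.
# SelectAbility lists Summon under Priority ahead of a Random group of BeamSweep, AppendageSlam and HomingProjectiles, each gated by Can Activate Ability.
# Completed ability tasks lead to Wait, whose Active state sets target focus and reselects once Ability.SB.Active clears.
# The Exploding tag diverts Active into an indefinite delay with no outgoing transitions.
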StateTree /SpireBoss/StateTrees/ST_Large_Spire_Combat
states [
 Combat {
  id=#1
  type=STATE
  children [#2, #3]
  tasks [
   {
    id=#4
    description="Debug Text \"COMBAT\""
    properties {
     iD=#5
     instance {_structType=/Script/StateTreeModule.StateTreeDebugTextTaskInstanceData, bEnabled=true, bindableText="", referenceActor=null}
     node {
      name=None
      _structType=/Script/StateTreeModule.StateTreeDebugTextTask
      bCanEditConsideredForCompletion=false
      bConsideredForCompletion=false
      bHasTaskCompletionDelegateDispatcher=false
      bTaskEnabled=false
      deprecatedVersion=1
      fontScale=1
      text=COMBAT
      transitionHandlingPriority=Normal
      bindingsBatch {value=65535}
      executionRuntimeTemplateIndex {value=65535}
      instanceDataHandle {index=65535, source=None, stateHandle {index=65535}}
      instanceTemplateIndex {value=65535}
      offset=<0 0 0>
      outputBindingsBatch {value=65535}
      textColor=<rgba 1 0 0.9765 1>
     }
    }
   }
   {
    id=#6
    description="(Fortnite) Toggle GameplayEffects"
    properties {
     iD=#7
     instance {
      _structType=/Script/FortniteAIServer.FortAthenaToggleGameplayEffectStateTreeTaskInstanceData
      actor=null
      bAutomaticallyRemoveAddedEffectsOnExit=false
      targetActor=null
      gameplayEffectClassesToAdd [/SpireBoss/Abilities/GE_Large_Spire_InitialCombat_Cooldown.GE_Large_Spire_InitialCombat_Cooldown_C]
      gameplayEffectClassesToRemove []
      gameplayEffectsByAppliedTagsToRemove {gameplayTags []}
      gameplayEffectsByGrantedTagsToRemove {gameplayTags []}
      gameplayEffectsBySourceTagsToRemove {gameplayTags []}
      gameplayEffectsByTagsToRemove {gameplayTags []}
     }
     node {
      name=None
      _structType=/Script/FortniteAIServer.FortAthenaToggleGameplayEffectStateTreeTask
      addExecutionMode=ExecuteOnEnter
      bCanEditConsideredForCompletion=true
      bConsideredForCompletion=false
      bHasTaskCompletionDelegateDispatcher=false
      bTaskEnabled=true
      removeExecutionMode=DoNotExecute
      transitionHandlingPriority=Normal
      bindingsBatch {value=65535}
      executionRuntimeTemplateIndex {value=65535}
      instanceDataHandle {index=65535, source=None, stateHandle {index=65535}}
      instanceTemplateIndex {value=65535}
      outputBindingsBatch {value=65535}
     }
    }
   }
   {
    id=#8
    description=STBTT Spire Run Encounter
    properties {
     iD=#9
     instanceObject=&self:StateTreeEditorData_0.StateTreeState_0.STBTT_Spire_RunEncounter_C_0
     node {
      name=None
      _structType=&StateTreeBlueprintTaskWrapper
      taskClass=/SpireBoss/StateTrees/Tasks/STBTT_Spire_RunEncounter.STBTT_Spire_RunEncounter_C
     }
    }
   }
   {
    id=#10
    description=State Tree Task Spires Set State
    properties {
     iD=#11
     instanceObject=&self:StateTreeEditorData_0.StateTreeState_0.StateTreeTask_Spires_SetState_C_0
     node {
      name=None
      _structType=&StateTreeBlueprintTaskWrapper
      taskClass=/SpireBoss/StateTrees/Tasks/StateTreeTask_Spires_SetState.StateTreeTask_Spires_SetState_C
     }
    }
   }
   {
    id=#12
    description=STBTT Spire Update Shields
    properties {
     iD=#13
     instanceObject=&self:StateTreeEditorData_0.StateTreeState_0.STBTT_Spire_UpdateShields_C_0
     node {
      name=None
      _structType=&StateTreeBlueprintTaskWrapper
      taskClass=/SpireBoss/StateTrees/Tasks/STBTT_Spire_UpdateShields.STBTT_Spire_UpdateShields_C
     }
    }
   }
  ]
  transitions []
 }
 SelectAbility {
  id=#2
  type=STATE
  children [#14, #15]
  tasks []
  transitions []
 }
 Wait {
  id=#3
  type=STATE
  children [#16, #17]
  tasks []
  transitions []
 }
 Priority {id=#14, type=STATE, children [#18], tasks [], transitions []}
 Random {
  id=#15
  type=STATE
  children [#19, #20, #21]
  tasks []
  transitions []
 }
 Exploding {
  id=#16
  type=STATE
  enter_conditions [{
    id=#22
    description="(Fortnite) Actor Has Matching Gameplay Tag"
    properties {
     iD=#23
     instance {
      _structType=/Script/FortniteAIServer.GameplayTagActorHasMatchingGameplayTagInstanceData
      actor=null
      tagToCheck {tagName=Gameplay.Status.Spire.Exploding}
     }
     node {
      name=None
      _structType=/Script/FortniteAIServer.GameplayTagActorHasMatchingGameplayTagCondition
      bInvert=false
      deltaIndent=0
      evaluationMode=Evaluated
      operand=And
      bindingsBatch {value=65535}
      executionRuntimeTemplateIndex {value=65535}
      instanceDataHandle {index=65535, source=None, stateHandle {index=65535}}
      instanceTemplateIndex {value=65535}
      outputBindingsBatch {value=65535}
     }
    }
   }]
  tasks [{
    id=#24
    description=Delay Forever
    properties {
     iD=#25
     instance {_structType=&StateTreeDelayTaskInstanceData, bRunForever=true, duration=1}
     node {name=None, _structType=&StateTreeDelayTask}
    }
   }]
  transitions []
 }
 Active {
  id=#17
  type=STATE
  enter_conditions [{
    id=#26
    description="(Fortnite) Actor Has Matching Gameplay Tag"
    properties {
     iD=#27
     instance {
      _structType=&GameplayTagActorHasMatchingGameplayTagInstanceData
      actor=null
      tagToCheck {tagName=Gameplay.Status.Spire.Exploding}
     }
     node {
      name=None
      _structType=&GameplayTagActorHasMatchingGameplayTagCondition
      bInvert=true
      deltaIndent=0
      evaluationMode=Evaluated
      operand=And
      bindingsBatch {value=65535}
      executionRuntimeTemplateIndex {value=65535}
      instanceDataHandle {index=65535, source=None, stateHandle {index=65535}}
      instanceTemplateIndex {value=65535}
      outputBindingsBatch {value=65535}
     }
    }
   }]
  tasks [
   {
    id=#28
    description=Delay Forever
    properties {
     iD=#29
     instance {_structType=&StateTreeDelayTaskInstanceData, bRunForever=true, duration=0}
     node {name=None, _structType=&StateTreeDelayTask}
    }
   }
   {
    id=#30
    description="[ActorFocus] Set Target Focus"
    properties {
     iD=#31
     instance {_structType=/Script/FortniteAI.FortActorFocusComponentSetActorFocusInstanceData, actor=null, focusActor=null}
     node {
      name=None
      _structType=/Script/FortniteAI.FortActorFocusComponentSetActorFocusTask
      bCanEditConsideredForCompletion=true
      bConsideredForCompletion=true
      bHasTaskCompletionDelegateDispatcher=false
      bTaskEnabled=true
      transitionHandlingPriority=Normal
      bindingsBatch {value=65535}
      executionRuntimeTemplateIndex {value=65535}
      instanceDataHandle {index=65535, source=None, stateHandle {index=65535}}
      instanceTemplateIndex {value=65535}
      outputBindingsBatch {value=65535}
     }
    }
   }
  ]
  transitions [
   {
    id=#32
    trigger=ON_TICK
    conditions [
     {
      id=#33
      description="(Fortnite) Actor Has Matching Gameplay Tag"
      properties {
       iD=#34
       instance {_structType=&GameplayTagActorHasMatchingGameplayTagInstanceData, actor=null, tagToCheck {tagName=Ability.SB.Active}}
       node {
        name=None
        _structType=&GameplayTagActorHasMatchingGameplayTagCondition
        bInvert=true
        deltaIndent=0
        evaluationMode=Evaluated
        operand=And
        bindingsBatch {value=65535}
        executionRuntimeTemplateIndex {value=65535}
        instanceDataHandle {index=65535, source=None, stateHandle {index=65535}}
        instanceTemplateIndex {value=65535}
        outputBindingsBatch {value=65535}
       }
      }
     }
     {
      id=#35
      description="(Fortnite) Actor Has Matching Gameplay Tag"
      properties {
       iD=#36
       instance {
        _structType=&GameplayTagActorHasMatchingGameplayTagInstanceData
        actor=null
        tagToCheck {tagName=Gameplay.Status.Spire.Exploding}
       }
       node {
        name=None
        _structType=&GameplayTagActorHasMatchingGameplayTagCondition
        bInvert=true
        deltaIndent=0
        evaluationMode=Evaluated
        operand=And
        bindingsBatch {value=65535}
        executionRuntimeTemplateIndex {value=65535}
        instanceDataHandle {index=65535, source=None, stateHandle {index=65535}}
        instanceTemplateIndex {value=65535}
        outputBindingsBatch {value=65535}
       }
      }
     }
    ]
    state {id=#2, name=SelectAbility}
   }
   {
    id=#37
    trigger=ON_TICK
    conditions [{
      id=#38
      description="(Fortnite) Actor Has Matching Gameplay Tag"
      properties {
       iD=#39
       instance {
        _structType=&GameplayTagActorHasMatchingGameplayTagInstanceData
        actor=null
        tagToCheck {tagName=Gameplay.Status.Spire.Exploding}
       }
       node {
        name=None
        _structType=&GameplayTagActorHasMatchingGameplayTagCondition
        bInvert=false
        deltaIndent=0
        evaluationMode=Evaluated
        operand=And
        bindingsBatch {value=65535}
        executionRuntimeTemplateIndex {value=65535}
        instanceDataHandle {index=65535, source=None, stateHandle {index=65535}}
        instanceTemplateIndex {value=65535}
        outputBindingsBatch {value=65535}
       }
      }
     }]
    state {id=#16, name=Exploding}
   }
  ]
 }
 Summon {
  id=#18
  type=STATE
  enter_conditions [{
    id=#40
    description=Can Activate Ability Ability.SB.Summon
    properties {
     iD=#41
     instance {
      _structType=/Script/FortniteAIServer.FortGameplayAbilityCanActivateAbilityInstanceData
      pawn=null
      abilityTag {gameplayTags [{tagName=Ability.SB.Summon}]}
     }
     node {
      name=None
      _structType=/Script/FortniteAIServer.FortGameplayAbilityCanActivateAbilityCondition
      bInvert=false
      deltaIndent=0
      evaluationMode=Evaluated
      operand=And
      bindingsBatch {value=65535}
      executionRuntimeTemplateIndex {value=65535}
      instanceDataHandle {index=65535, source=None, stateHandle {index=65535}}
      instanceTemplateIndex {value=65535}
      outputBindingsBatch {value=65535}
     }
    }
   }]
  tasks [{
    id=#42
    description=Start Ability Ability.SB.Summon
    properties {
     iD=#43
     instance {
      _structType=/Script/FortniteGame.FortStateTreeStartAbilityTaskInstanceData
      actor=null
      abilityTag {gameplayTags [{tagName=Ability.SB.Summon}]}
     }
     node {
      name=None
      _structType=/Script/FortniteGame.FortStateTreeStartAbilityTask
      bCanEditConsideredForCompletion=true
      bConsideredForCompletion=true
      bHasTaskCompletionDelegateDispatcher=false
      bTaskEnabled=true
      transitionHandlingPriority=Normal
      bindingsBatch {value=65535}
      executionRuntimeTemplateIndex {value=65535}
      instanceDataHandle {index=65535, source=None, stateHandle {index=65535}}
      instanceTemplateIndex {value=65535}
      outputBindingsBatch {value=65535}
     }
    }
   }]
  transitions [{id=#44, trigger=ON_STATE_COMPLETED, state {id=#3, name=Wait}}]
 }
 BeamSweep {
  id=#19
  type=STATE
  enter_conditions [{
    id=#45
    description=Can Activate Ability Ability.SB.BeamSweep
    properties {
     iD=#46
     instance {
      _structType=&FortGameplayAbilityCanActivateAbilityInstanceData
      pawn=null
      abilityTag {gameplayTags [{tagName=Ability.SB.BeamSweep}]}
     }
     node {
      name=None
      _structType=&FortGameplayAbilityCanActivateAbilityCondition
      bInvert=false
      deltaIndent=0
      evaluationMode=Evaluated
      operand=And
      bindingsBatch {value=65535}
      executionRuntimeTemplateIndex {value=65535}
      instanceDataHandle {index=65535, source=None, stateHandle {index=65535}}
      instanceTemplateIndex {value=65535}
      outputBindingsBatch {value=65535}
     }
    }
   }]
  tasks [{
    id=#47
    description=Start Ability Ability.SB.BeamSweep
    properties {
     iD=#48
     instance {
      _structType=&FortStateTreeStartAbilityTaskInstanceData
      actor=null
      abilityTag {gameplayTags [{tagName=Ability.SB.BeamSweep}]}
     }
     node {
      name=None
      _structType=&FortStateTreeStartAbilityTask
      bCanEditConsideredForCompletion=true
      bConsideredForCompletion=true
      bHasTaskCompletionDelegateDispatcher=false
      bTaskEnabled=true
      transitionHandlingPriority=Normal
      bindingsBatch {value=65535}
      executionRuntimeTemplateIndex {value=65535}
      instanceDataHandle {index=65535, source=None, stateHandle {index=65535}}
      instanceTemplateIndex {value=65535}
      outputBindingsBatch {value=65535}
     }
    }
   }]
  transitions [{id=#49, trigger=ON_STATE_COMPLETED, state {id=#3, name=Wait}}]
 }
 AppendageSlam {
  id=#20
  type=STATE
  enter_conditions [{
    id=#50
    description=Can Activate Ability Ability.SB.AppendageSlam
    properties {
     iD=#51
     instance {
      _structType=&FortGameplayAbilityCanActivateAbilityInstanceData
      pawn=null
      abilityTag {gameplayTags [{tagName=Ability.SB.AppendageSlam}]}
     }
     node {
      name=None
      _structType=&FortGameplayAbilityCanActivateAbilityCondition
      bInvert=false
      deltaIndent=0
      evaluationMode=Evaluated
      operand=And
      bindingsBatch {value=65535}
      executionRuntimeTemplateIndex {value=65535}
      instanceDataHandle {index=65535, source=None, stateHandle {index=65535}}
      instanceTemplateIndex {value=65535}
      outputBindingsBatch {value=65535}
     }
    }
   }]
  tasks [{
    id=#52
    description=Start Ability Ability.SB.AppendageSlam
    properties {
     iD=#53
     instance {
      _structType=&FortStateTreeStartAbilityTaskInstanceData
      actor=null
      abilityTag {gameplayTags [{tagName=Ability.SB.AppendageSlam}]}
     }
     node {
      name=None
      _structType=&FortStateTreeStartAbilityTask
      bCanEditConsideredForCompletion=true
      bConsideredForCompletion=true
      bHasTaskCompletionDelegateDispatcher=false
      bTaskEnabled=true
      transitionHandlingPriority=Normal
      bindingsBatch {value=65535}
      executionRuntimeTemplateIndex {value=65535}
      instanceDataHandle {index=65535, source=None, stateHandle {index=65535}}
      instanceTemplateIndex {value=65535}
      outputBindingsBatch {value=65535}
     }
    }
   }]
  transitions [{id=#54, trigger=ON_STATE_COMPLETED, state {id=#3, name=Wait}}]
 }
 HomingProjectiles {
  id=#21
  type=STATE
  enter_conditions [{
    id=#55
    description=Can Activate Ability Ability.SB.HomingProjectiles
    properties {
     iD=#56
     instance {
      _structType=&FortGameplayAbilityCanActivateAbilityInstanceData
      pawn=null
      abilityTag {gameplayTags [{tagName=Ability.SB.HomingProjectiles}]}
     }
     node {
      name=None
      _structType=&FortGameplayAbilityCanActivateAbilityCondition
      bInvert=false
      deltaIndent=0
      evaluationMode=Evaluated
      operand=And
      bindingsBatch {value=65535}
      executionRuntimeTemplateIndex {value=65535}
      instanceDataHandle {index=65535, source=None, stateHandle {index=65535}}
      instanceTemplateIndex {value=65535}
      outputBindingsBatch {value=65535}
     }
    }
   }]
  tasks [{
    id=#57
    description=Start Ability Ability.SB.HomingProjectiles
    properties {
     iD=#58
     instance {
      _structType=&FortStateTreeStartAbilityTaskInstanceData
      actor=null
      abilityTag {gameplayTags [{tagName=Ability.SB.HomingProjectiles}]}
     }
     node {
      name=None
      _structType=&FortStateTreeStartAbilityTask
      bCanEditConsideredForCompletion=true
      bConsideredForCompletion=true
      bHasTaskCompletionDelegateDispatcher=false
      bTaskEnabled=true
      transitionHandlingPriority=Normal
      bindingsBatch {value=65535}
      executionRuntimeTemplateIndex {value=65535}
      instanceDataHandle {index=65535, source=None, stateHandle {index=65535}}
      instanceTemplateIndex {value=65535}
      outputBindingsBatch {value=65535}
     }
    }
   }]
  transitions [{id=#59, trigger=ON_STATE_COMPLETED, state {id=#3, name=Wait}}]
 }
]
