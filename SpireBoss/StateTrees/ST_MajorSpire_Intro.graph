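# State tree that runs the Major Spire boss's intro sequence.
# Trigger Intro Animation starts the gameplay ability tagged Ability.MajorSpire.IntroSequence and advances to the next state when it completes.
# Destroy Buildings then runs a Blueprint task that destroys buildings on the spire's plunge, and the tree succeeds on completion.
# An event transition on the first state can also end the tree as succeeded, skipping the destruction step.
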StateTree /SpireBoss/StateTrees/ST_MajorSpire_Intro
states [
 Root {
  id=#1
  type=STATE
  children [#2, #3]
  tasks []
  transitions []
 }
 Trigger Intro Animation {
  id=#2
  type=STATE
  tasks [{
    id=#4
    description=Start Ability Ability.MajorSpire.IntroSequence
    properties {
     iD=#5
     instance {
      _structType=/Script/FortniteGame.FortStateTreeStartAbilityTaskInstanceData
      actor=null
      abilityTag {gameplayTags [{tagName=Ability.MajorSpire.IntroSequence}]}
     }
     node {
      name=None
      _structType=/Script/FortniteGame.FortStateTreeStartAbilityTask
      bCanEditConsideredForCompletion=true
      bConsideredForCompletion=true
      bHasTaskCompletionDelegateDispatcher=false
      bTaskEnabled=true
      transitionHandlingPriority=Normal
      bindingsBatch {value=65535}
      executionRuntimeTemplateIndex {value=65535}
      instanceDataHandle {index=65535, source=None, stateHandle {index=65535}}
      instanceTemplateIndex {value=65535}
      outputBindingsBatch {value=65535}
     }
    }
   }]
  transitions [
   {id=#6, trigger=ON_STATE_COMPLETED, state {id=#7, name=None, linkType=NEXT_STATE}}
   {id=#8, trigger=ON_EVENT, state {id=#7, name=None, linkType=SUCCEEDED}}
  ]
 }
 Destroy Buildings {
  id=#3
  type=STATE
  tasks [{
    id=#9
    description=STBTT Spire Destroy Buildings on Plunge
    properties {
     iD=#10
     instanceObject=&self:StateTreeEditorData_0.StateTreeState_2.STBTT_Spire_DestroyBuildingsOnPlunge_C_0
     node {
      name=None
      _structType=&StateTreeBlueprintTaskWrapper
      taskClass=/SpireBoss/StateTrees/Tasks/STBTT_Spire_DestroyBuildingsOnPlunge.STBTT_Spire_DestroyBuildingsOnPlunge_C
     }
    }
   }]
  transitions [{id=#11, trigger=ON_STATE_COMPLETED, state {id=#7, name=Root, linkType=SUCCEEDED}}]
 }
]
